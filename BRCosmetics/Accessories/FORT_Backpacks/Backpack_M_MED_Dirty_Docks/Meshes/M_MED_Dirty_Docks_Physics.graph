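# Physics asset for the Dirty Docks backpack cosmetic, giving capsule bodies and joint limits to its dangling parts.
# A pack body anchors a five-segment flap chain plus separate left knife, right knife and zipper bodies.
# Flap joints allow generally wider swing toward the tip, the knives stay tightly limited, and the zipper swings more freely.
# All joints share a narrow twist limit.
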PhysicsAsset /BRCosmetics/Accessories/FORT_Backpacks/Backpack_M_MED_Dirty_Docks/Meshes/M_MED_Dirty_Docks_Physics
bodies {
 dyn_flap_01 [{shapeType=Capsule, length=2.5, radius=4, center=<0 -1.0045 0>, rotation=<rot 90 0 0>}]
 dyn_flap_02 [{shapeType=Capsule, length=5.25, radius=3.25, center=<0.5 -2.2712 0>, rotation=<rot 90 0 0>}]
 dyn_flap_03 [{shapeType=Capsule, length=2.75, radius=2.5, center=<0.85 -1.5379 0>, rotation=<rot 90 0 0>}]
 dyn_flap_04 [{shapeType=Capsule, length=2, radius=2.25, center=<1.5 -1.2472 0>, rotation=<rot 90 0 0>}]
 dyn_flap_05 [{shapeType=Capsule, length=2.5, radius=1.5, center=<2 -0.7399 0>, rotation=<rot 90 0 0>}]
 dyn_knife_l [{shapeType=Capsule, length=3.75, radius=2.75, center=<0.35 0 0.0001>, rotation=<rot 79.1521 -179.9999 -179.9999>}]
 dyn_knife_r [{shapeType=Capsule, length=3.75, radius=3, center=<0.8 0 0>, rotation=<rot 90 0 0>}]
 dyn_zipper [{shapeType=Capsule, length=3.5, radius=2.005, center=<1.0293 -0.006 -0.1263>, rotation=<rot 0 90.0156 -89.1851>}]
 pack [{shapeType=Capsule, length=15.5, radius=10.3386, center=<0.5104 -0.146 -2.6064>, rotation=<rot 2.2806 90.1187 -178.2672>}]
}
constraints [
 {bone1Name=dyn_knife_l, bone2Name=pack, swing1LimitDegrees=9, swing2LimitDegrees=8, twistLimitDegrees=5}
 {bone1Name=dyn_knife_r, bone2Name=pack, swing1LimitDegrees=10, swing2LimitDegrees=10, twistLimitDegrees=5}
 {bone1Name=dyn_zipper, bone2Name=pack, swing1LimitDegrees=40, swing2LimitDegrees=25, twistLimitDegrees=5}
 {bone1Name=dyn_flap_02, bone2Name=dyn_flap_01, swing1LimitDegrees=20, swing2LimitDegrees=20, twistLimitDegrees=5}
 {bone1Name=dyn_flap_03, bone2Name=dyn_flap_02, swing1LimitDegrees=20, swing2LimitDegrees=20, twistLimitDegrees=5}
 {bone1Name=dyn_flap_04, bone2Name=dyn_flap_03, swing1LimitDegrees=35, swing2LimitDegrees=25, twistLimitDegrees=5}
 {bone1Name=dyn_flap_05, bone2Name=dyn_flap_04, swing1LimitDegrees=45, swing2LimitDegrees=25, twistLimitDegrees=5}
 {bone1Name=dyn_flap_01, bone2Name=pack, swing1LimitDegrees=19.8571, swing2LimitDegrees=26.7143, twistLimitDegrees=5}
]
